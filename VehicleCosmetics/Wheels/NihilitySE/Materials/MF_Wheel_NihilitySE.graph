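# Material function for the NihilitySE wheel cosmetic that adds an emissive infinity effect to the rim, on top of MF_WheelParent_Simple.
# A UV1 infinity texture is summed with MF_Infinity_OffsetFunction's bump-offset result, with the rim mask's blue channel choosing red over green.
# CustomColor and EmissivePower tint and scale it within the green and blue mask areas.
# Red vertex color restores the parent's emissive; other attributes pass through.
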MaterialFunction /VehicleCosmetics/Wheels/NihilitySE/Materials/MF_Wheel_NihilitySE
graph {nodes [
  n1 = MaterialFunctionCall() {materialFunction=&MF_WheelParent_Simple}
  n2 = GetMaterialAttributes(None=@n1.MaterialAttributes) {attributeGetTypes [#1]}
  n3 = TextureObjectParameter() {
   constCoordinate=1
   group=0_InputAni
   parameterName=Inifiny_Texture
   sortPriority=1
   texture=/VehicleCosmetics/Wheels/NihilitySE/Textures/T_Rim_NihilitySE_Infinity
  }
  n4 = TextureSample(Tex=@n3) {constCoordinate=1}
  n5 = MaterialFunctionCall!(
   Texture=@n3
   Offset=ScalarParameter() {defaultValue=-0.5, group=0_InputAni, parameterName=BumpOffsetAmount_RimInf}
   RotateUVs
   GradientDarkness=ScalarParameter() {defaultValue=0.032, group=0_InputAni, parameterName=GradientDarkness_RimInf}
   GradientFalloff
   UVs=TextureCoordinate() {coordinateIndex=1}
   Opt_LowQualityOffset=ScalarParameter() {defaultValue=200, parameterName="Opt_LowQualityOffset (S)"}
   Opt_MediumQualityOffset
  ) {materialFunction=/VehicleCosmetics/SharedMaterials/MaterialFunctions/MF_Infinity_OffsetFunction}
  n6 = ComponentMask(None=@n1.RimRGBAMask) {b=true}
  n7 = LinearInterpolate(A=Add(A=@n4.G, B=ComponentMask(None=@n5.Result) {g=true}), B=Add(A=@n4.R, B=ComponentMask(None=@n5.Result) {r=true}), Alpha=@n6)
  n8 = ScalarParameter() {defaultValue=3, group=0_InputAni, parameterName=EmissivePower}
  n9 = LinearInterpolate(
   A=@n2.EmissiveColor
   B=Multiply(A=VectorParameter() {group=0_InputAni, parameterName=CustomColor, defaultValue=<rgba 0.7176 0.3319 1 1>}.RGB, B=Multiply(A=@n7, B=@n8))
   Alpha=Add(A=ComponentMask(None=@n1.RimRGBAMask) {g=true}, B=@n6)
  )
  n10 = VertexColor()
  FunctionOutput(None=SetMaterialAttributes!(MaterialAttributes=@n2.MaterialAttributes, EmissiveColor=LinearInterpolate(A=@n9, B=@n2.EmissiveColor, Alpha=@n10.R)) {attributeSetTypes [#1]})
 ]}
